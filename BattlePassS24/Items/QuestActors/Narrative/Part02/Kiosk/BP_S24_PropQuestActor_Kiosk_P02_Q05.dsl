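; Season 24 quest kiosk prop: an arcade-style grid puzzle where dialogue choices move a green dot on screen.
; Conversation node metadata gives position and action (move, error, win), driving scalar parameters on the screen's dynamic material plus sounds.
; Construction applies materials, audio cues and this instance's end point and obstacles.
; Feedback runs only for the local client, split-screen aware; interaction and overlap changes toggle the screen state and loop audio.
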(event OnReceivedLocalClientConversationMessage(NonPlayerConversationComponent)
  (params (LocalPlayerController :type Fort%20Player%20Controller%20Object%20Reference) (Message_Message :type Client%20Conversation%20Message%20Structure) (Message_Participants :type Conversation%20Participants%20Structure) (Message_CurrentNode :type Conversation%20Node%20Handle%20Structure) (Message_Options :type Array%20of%20Client%20Conversation%20Option%20Entry%20Structures))
  (bind (_speakerid _participantdisplayname _text _metadataparameters) (Utilities|Struct|BreakClientConversationMessage Message_Message))
  (bind _output (Utilities|Array|Get(aref) _metadataparameters))
  (bind (_name _value) (Utilities|Struct|BreakConversationNodeParameterPair _output))
  ;[0 0 0] When a conversation choice is selected, get the metadata on that node and parse it
  (bind _output_2 (CallFunction|IsValidLocalPlayerController LocalPlayerController))
  (if _output_2
    (Variables|Default|SetPreviousInteractionState (Variables|Default|GetCurrent%20Interaction%20State))
    (Rendering|Material|SetScalarParameterValue (Variables|Default|GetScreenMID) "use error" 0.0 :!class /Script/Engine.MaterialInstanceDynamic)
    (Variables|Default|SetPlayerPosition (Utilities|String|StringToFloat _name))
    (if (and (not (Utilities|String|IsEmpty _name)) (not (Utilities|String|IsEmpty _value)))
      ;[0 0 0] Change material parameter based on metadata. Name = the current green dot index number, Value = the parameter string to set
      (switch string _value
        (:Default
          (Rendering|Material|SetScalarParameterValue (Variables|Default|GetScreenMID) "Green dot index" (Variables|Default|GetGreenDotStart) :!class /Script/Engine.MaterialInstanceDynamic)
          (Rendering|Material|SetScalarParameterValue (Variables|Default|GetScreenMID) "use error" 1.0 :!class /Script/Engine.MaterialInstanceDynamic)
          (Utilities|FlowControl|Delay 0.3)
          (Rendering|Material|SetScalarParameterValue (Variables|Default|GetScreenMID) "use error" 0.0 :!class /Script/Engine.MaterialInstanceDynamic))
        (:Up
          (Rendering|Material|SetScalarParameterValue (Variables|Default|GetScreenMID) "Green dot index" (- (Variables|Default|GetPlayerPosition) 5.0) :!class /Script/Engine.MaterialInstanceDynamic))
        (:Down
          (Rendering|Material|SetScalarParameterValue (Variables|Default|GetScreenMID) "Green dot index" (+ (Variables|Default|GetPlayerPosition) 5.0) :!class /Script/Engine.MaterialInstanceDynamic))
        (:Left
          (Rendering|Material|SetScalarParameterValue (Variables|Default|GetScreenMID) "Green dot index" (- (Variables|Default|GetPlayerPosition) 1.0) :!class /Script/Engine.MaterialInstanceDynamic))
        (:Right
          (Rendering|Material|SetScalarParameterValue (Variables|Default|GetScreenMID) "Green dot index" (+ (Variables|Default|GetPlayerPosition) 1.0) :!class /Script/Engine.MaterialInstanceDynamic))
        (:Win
          (Rendering|Material|SetScalarParameterValue (Variables|Default|GetScreenMID) "Use lock unlocking" 1.0 :!class /Script/Engine.MaterialInstanceDynamic)
          (Variables|Default|SetbHasWon true)
          (Audio|Components|Audio|FadeOut (Variables|Components|GetLoopSoundComponent) (Variables|Design|PreInteraction|AmbientFX|GetAmbientAudioFadeOut) 0.0 "Linear" :!class /Script/Engine.AudioComponent)
          (Audio|SpawnSoundatLocation (Variables|AudioConfig|GetWinSoundCue) (Transformation|GetWorldLocation (Variables|Transformation|GetRootComponent)))
          (Utilities|FlowControl|Delay 1.0)
          (Rendering|Material|SetScalarParameterValue (Variables|Default|GetScreenMID) "unlocking" 1.0 :!class /Script/Engine.MaterialInstanceDynamic))
        (:Error
          (Rendering|Material|SetScalarParameterValue (Variables|Default|GetScreenMID) "use error" 1.0 :!class /Script/Engine.MaterialInstanceDynamic)
          (CallFunction|PlaySound (Variables|Components|GetErrorSoundComponent))))
      (CallFunction|PlaySound (Variables|Components|GetValidMoveSoundComponent)))))

(event Custom|PlaySound
  (params (SoundToPlay :type Audio%20Component%20Object%20Reference) (bShouldDelay :type Boolean))
  ;[0 0 0] Client only - play a specified sound cue at the prop's location
  (Utilities|IsValid SoundToPlay
    (:Is%20Valid
      (if (not (Audio|Components|Audio|IsPlaying SoundToPlay))
        (if bShouldDelay
          (Utilities|FlowControl|Delay 0.6)
          (else
            _))
        (else
          (Audio|Components|Audio|FadeOut SoundToPlay 0.2 0.0 "Linear" :!class /Script/Engine.AudioComponent)))
      (Audio|Components|Audio|Play SoundToPlay 0.0))))

(event OnDialogMarkerInteractionStateChanged(NonPlayerConversationComponent)
  (params (CurrentInteractionState :type EDialogMarkerInteractionState%20Enum :default "Conversation") (InteractingController :type Fort%20Player%20Controller%20Object%20Reference))
  ;[0 0 0] Change the state of the screen when the local player starts/stops conversing with the prop
  (bind _output_get (Variables|Default|SetCurrent%20Interaction%20State CurrentInteractionState))
  (Utilities|FlowControl|Sequence
    (:then_0
      (switch Utilities|FlowControl|Switch|SwitchonEDialogMarkerInteractionState _output_get
        (:Conversation
          _)
        (:InteractionRange
          (Networking|SwitchHasAuthority
            (:Remote
              (bind _output (CallFunction|IsValidLocalPlayerController InteractingController))
              (if _output
                (Rendering|Material|SetScalarParameterValue (Variables|Default|GetScreenMID) "use win" 1.0 :!class /Script/Engine.MaterialInstanceDynamic)
                (Audio|Components|Audio|FadeOut (Variables|Components|GetLoopSoundComponent) (Variables|Design|PreInteraction|AmbientFX|GetAmbientAudioFadeOut) 0.0 "Linear" :!class /Script/Engine.AudioComponent)))))
        (:IndicatorRange
          _)
        (:None
          _))
      (Networking|SwitchHasAuthority
        (:Remote
          (CallFunction|IsValidLocalPlayerController InteractingController)
          (if (not (Variables|Default|GetbHasWon))
            (Rendering|Material|SetScalarParameterValue (Variables|Default|GetScreenMID) "use win" 0.0 :!class /Script/Engine.MaterialInstanceDynamic)))))))

(event EventResetforVisualTesting
  (params (PlayerController :type Fort%20Player%20Controller%20Object%20Reference))
  ;[0 0 0] If prop is in visual test mode, undo the endgame screen state on prop reset
  (|CallParentFunction PlayerController)
  (bind _output (CallFunction|IsValidLocalPlayerController PlayerController))
  (if _output
    (Variables|Default|SetbHasWon false)))

(event OnComponentEndOverlap(ClientEventsBounds)
  (params (OverlappedComponent :type Primitive%20Component%20Object%20Reference) (OtherActor :type Actor%20Object%20Reference) (OtherComp :type Primitive%20Component%20Object%20Reference) (OtherBodyIndex :type Integer))
  ;[0 0 0] Client only - Turn off the music and reset the screen visuals when the player leaves the prop
  (Networking|SwitchHasAuthority
    (:Remote
      (bind _asfort_player_pawn (Utilities|Casting|CastToFortPlayerPawn OtherActor))
      (bind _output (CallFunction|IsValidLocalPlayerController (Fortnite|GetFortPlayerControllerfromActor _asfort_player_pawn)))
      (if _output
        (Rendering|Material|SetScalarParameterValue (Variables|Default|GetScreenMID) "use win" 1.0 :!class /Script/Engine.MaterialInstanceDynamic)
        (Audio|Components|Audio|FadeOut (Variables|Components|GetLoopSoundComponent) (Variables|Design|PreInteraction|AmbientFX|GetAmbientAudioFadeOut) 0.0 "Linear" :!class /Script/Engine.AudioComponent)))))

(fn BlueprintOnLocalInteract
  (params (InteractingPawn :type Fort%20Player%20Pawn%20Object%20Reference) (InteractionBeingAttempted :type EInteractionBeingAttempted%20Enum :default "FirstInteraction"))
  (bind _returnvalue (|CallParentFunction InteractingPawn InteractionBeingAttempted))
  (bind _output (CallFunction|IsValidLocalPlayerController (Fortnite|GetFortPlayerControllerfromActor InteractingPawn)))
  (if _output
    (CallFunction|PlaySound (Variables|Components|GetStartSoundComponent))
    (CallFunction|PlaySound (Variables|Components|GetLoopSoundComponent) true)
    (return _returnvalue)))

(fn IsValidLocalPlayerController
  (params (InstigatingPlayerController :type Fort%20Player%20Controller%20Object%20Reference))
  ;[0 0 0] Returns true on the executing client if the instigator controller is locally controlled. Returns false if running on the server
  (if (not (Networking|IsServer))
    (Utilities|IsValid InstigatingPlayerController
      (:Is%20Valid
        (if (PlayerController|IsinSplitscreen InstigatingPlayerController)
          (for _array_element (Fortnite|GetLocalFortPlayerControllers true)
            (if (== _array_element InstigatingPlayerController)
              (return true)))
          (return false)
          (else
            (return true))))
      (:Is%20Not%20Valid
        _)))
  (return false))

(fn ConstructionScript
  (bind _screenmid (Variables|Default|GetScreenMID))
  ;[0 0 0] Set variables for the two arcade materials (screen and cabinet) so they are easily accesible from class defaults
  (|CallParentFunction)
  (Utilities|FlowControl|Sequence
    (:then_0
      (Rendering|Material|SetMaterial (Variables|Components|GetSM_Quest_PreInteraction) 1 (Variables|ArcadeConfig|GetScreenMaterial))
      (Rendering|Material|SetMaterial (Variables|Components|GetSM_Quest_PreInteraction) 0 (Variables|ArcadeConfig|GetCabinetMaterial))
      (bind _returnvalue (Rendering|Material|CreateDynamicMaterialInstance (Variables|Components|GetSM_Quest_PreInteraction) 1 (Variables|ArcadeConfig|GetScreenMaterial) "Screen" :!class /Script/Engine.PrimitiveComponent))
      (bind _output_get (Variables|Default|SetScreenMID _returnvalue))
      (Rendering|Material|SetScalarParameterValue _output_get "use win" 1.0 :!class /Script/Engine.MaterialInstanceDynamic))
    (:then_1
      ;;[1 0.3798 0.6811] Change these per unique instance
      (Rendering|Material|SetScalarParameterValue _screenmid "Green dot index" (Variables|Default|GetGreenDotStart) :!class /Script/Engine.MaterialInstanceDynamic)
      (Rendering|Material|SetScalarParameterValue _screenmid "End Point Index" 6.0 :!class /Script/Engine.MaterialInstanceDynamic)
      (Rendering|Material|SetScalarParameterValue _screenmid "obstacle index" 4.0 :!class /Script/Engine.MaterialInstanceDynamic)
      (Rendering|Material|SetScalarParameterValue _screenmid "obstacle2 index" 7.0 :!class /Script/Engine.MaterialInstanceDynamic)
      (Rendering|Material|SetScalarParameterValue _screenmid "obstacle3 index" 11.0 :!class /Script/Engine.MaterialInstanceDynamic)
      (Rendering|Material|SetScalarParameterValue _screenmid "obstacle4 index" 14.0 :!class /Script/Engine.MaterialInstanceDynamic))
    (:then_2
      (Audio|Components|Audio|SetSound (Variables|Components|GetStartSoundComponent) (Variables|AudioConfig|GetStartSoundCue))
      (Audio|Components|Audio|SetSound (Variables|Components|GetValidMoveSoundComponent) (Variables|AudioConfig|GetValidMoveSoundCue))
      (Audio|Components|Audio|SetSound (Variables|Components|GetErrorSoundComponent) (Variables|AudioConfig|GetErrorSoundCue))
      (Audio|Components|Audio|SetSound (Variables|Components|GetLoopSoundComponent) (Variables|AudioConfig|GetLoopSoundCue)))))
